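; PCG template subgraph for the core process step of a primitive, found in the PCGPrimitives template folder.
; A Filter Data By Tag node splits incoming data on the "Input:0" tag into primary and secondary inputs.
; Both the matching and non-matching branches connect straight to the single Out pin, so the unmodified template only passes data through, with no processing between.
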(graph Primitives_CoreProcess
  ;[0 0 0 0.5 12] Primary/Secondary Inputs
  (bind FilterDataByTag_0 (Filter%20Data%20By%20Tag :SelectedTags "Input:0"))
  (connect DefaultInputNode In FilterDataByTag_0 In)
  (output Out FilterDataByTag_0.InsideFilter)
  (output Out FilterDataByTag_0.OutsideFilter))
