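; PCG template graph for an inputs fallback in primitives, meant to create fallback data or flag that no valid input data is found.
; A Create Constant node feeds a Print String node that reports at warning verbosity on the node, prefixed with component and owner.
; The printed text comes from the overridable WarningMessage parameter, and neither node connects to the graph's input or output.
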(graph Primitives_InputsFallback
  (params (WarningMessage :type String :default "No Valid Input Data Found"))
  ;[0 0 0 0.5] Create Fallback Data or Warning
  (bind CreateConstant_0 (Create%20Constant))
  (bind Print%20String_1 (Print%20String :bDisplayOnNode true :bPrefixWithComponent true :bPrefixWithNode false :bPrefixWithOwner true :PrintString "No Valid Input Data" :Verbosity Warning :PrintString-param WarningMessage))
  (connect CreateConstant_0 Out Print%20String_1 In))
