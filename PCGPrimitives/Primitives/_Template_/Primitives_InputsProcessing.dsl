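; Primitives_InputsProcessing is a PCG graph in the Primitives template folder that serves as an input-processing stub.
; It declares typed inputs for points, splines, closed splines, 2D polygons, surfaces, textures and volumes, each with a matching output.
; No processing nodes sit between them, so each input passes straight through to the output of the same name.
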(graph Primitives_InputsProcessing
  (inputs (Points :type Point) (Splines :type Spline) (ClosedSplines :type Spline) (Polygon2Ds :type Polygon2D) (Surfaces :type Surface) (Texture2Ds :type BaseTexture) (Volumes :type Volume))
  (output Out ClosedSplines)
  (output Out Points)
  (output Out Polygon2Ds)
  (output Out Splines)
  (output Out Surfaces)
  (output Out Texture2Ds)
  (output Out Volumes))
