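; PCG subgraph in the PCGPrimitives template folder that prepares a primitive's output data before it leaves the graph.
; It removes empty data, then uses a data count to branch so that only non-empty results continue.
; ForwardAllInputTags picks the final step: deleting just the Input:0 and Input:1 tags, or running the shared OutputTaggingSG subgraph with DataType and Primitive.
; Both paths feed the same output.
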(graph Primitives_PrepareOutputs
  (params (DataType :type String :default "None") (Primitive :type String :default "None") (ForwardAllInputTags :type Bool))
  (bind RemoveEmptyData_0 (Remove%20Empty%20Data))
  (bind DataNum_1 (Data%20Count))
  (bind Branch_2 (Branch))
  (bind Branch_1 (Branch :bOutputToB-param ForwardAllInputTags))
  ;[0 0 0 0.5] Tag and apply Metadata for Output
  (bind Subgraph_0 (Subgraph :Graph /PCGPrimitives/Primitives/Shared/PrepareOutputs/OutputTaggingSG.OutputTaggingSG :DataType DataType :Primitive Primitive))
  (bind DeleteTags_0 (Delete%20Tags :SelectedTags "Input:0, Input:1"))
  (connect Branch_2 Output%20B Branch_1 In)
  (connect DataNum_1 Out Branch_2 bOutputToB)
  (connect RemoveEmptyData_0 Out Branch_2 In)
  (connect RemoveEmptyData_0 Out DataNum_1 In)
  (connect Branch_1 Output%20B DeleteTags_0 In)
  (connect DefaultInputNode In RemoveEmptyData_0 In)
  (connect Branch_1 Output%20A Subgraph_0 In)
  (output Out DeleteTags_0)
  (output Out Subgraph_0))
